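# Post-process style Anim Blueprint for the CherryKnot female body cosmetic, adding secondary motion to its left bracelet and belt accessories.
# It copies the attached parent's pose through a Control Rig, then chains five spring-driven AnimDynamics nodes, blending back belt bones partially.
# Each update computes Gravity Override from the CherryKnot_Strings parameter asset and a front-end-dependent alpha, though the anim nodes show fixed gravity and alphas.
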AnimBlueprint /BRCosmetics/Characters/Player/Female/Medium/Bodies/F_MED_CherryKnot/Meshes/F_MED_CherryKnot_AnimBP
target_skeleton=&/F_MED_CherryKnot_Skeleton
graphs [
 AnimGraph {nodes [
   n1 = Control Rig(Source=Copy Pose From Mesh() {bUseAttachedParent=true}) {controlRigAssetReference {controlRigAsset=&/F_MED_CherryKnot_CtrlRig}}
   n2 = Anim Dynamics - Bone: dyn_bracelet_l_MSTR!(ComponentPose=Local To Component(LocalPose=@n1), Alpha=1, GravityOverride:Vector="0,0,-980") -> (Pose) {
    angularDampingOverride=1
    angularSpringConstant=1000
    bAngularSpring=true
    bOverrideAngularDamping=true
    bUseGravityOverride=true
    gravityScale=0.35
    simulationSpace=BoneRelative
    boundBone {boneName=dyn_bracelet_l_MSTR}
    componentAppliedLinearAccClamp=<3000 3000 3000>
    componentLinearAccScale=<0.5 0.5 0.5>
    componentLinearVelScale=<0.5 0.5 0.5>
    gravityOverride {z=-980}
    physicsBodyDefinitions [{
      boundBone {boneName=dyn_bracelet_l_MSTR}
      boxExtents=<15 17 17>
      constraintSetup {angularLimitsMax=<5 10 10>, angularLimitsMin=<-5 -50 -10>, angularTarget=<1 0 0>}
      localJointOffset=<4 0 -1>
     }]
    relativeSpaceBone {boneName=lowerarm_l}
    simSpaceSettings {maxAngularAcceleration=100, maxAngularVelocity=10, simSpaceAngularAlpha=1}
   }
   n3 = Anim Dynamics - Bone: dyn_beltAcc_front_01!(ComponentPose=@n2, Alpha=1, GravityOverride:Vector="0,0,-980") -> (Pose) {
    angularDampingOverride=1
    angularSpringConstant=3000
    bAngularSpring=true
    bOverrideAngularDamping=true
    bUseGravityOverride=true
    gravityScale=0.35
    simulationSpace=RootRelative
    boundBone {boneName=dyn_beltAcc_front_01}
    componentAppliedLinearAccClamp=<3000 3000 3000>
    componentLinearAccScale=<0.5 0.5 0.5>
    componentLinearVelScale=<0.5 0.5 0.5>
    gravityOverride {z=-980}
    physicsBodyDefinitions [{
      boundBone {boneName=dyn_beltAcc_front_01}
      boxExtents=<15 15 15>
      constraintSetup {angularLimitsMax=<0 4 45>, angularLimitsMin=<0 -4 0>, angularTarget=<1 0 0>}
      localJointOffset=<4 0 0>
     }]
    relativeSpaceBone {boneName=pelvis}
    simSpaceSettings {maxAngularAcceleration=100, maxAngularVelocity=10, simSpaceAngularAlpha=1}
   }
   n4 = Anim Dynamics - Bone: dyn_beltAcc_back_00!(ComponentPose=@n3, Alpha=0.7, GravityOverride:Vector="0,0,-980") -> (Pose) {
    angularDampingOverride=1
    angularSpringConstant=1000
    bAngularSpring=true
    bOverrideAngularDamping=true
    bUseGravityOverride=true
    gravityScale=0.35
    simulationSpace=RootRelative
    boundBone {boneName=dyn_beltAcc_back_00}
    componentAppliedLinearAccClamp=<3000 3000 3000>
    componentLinearAccScale=<0.5 0.5 0.5>
    componentLinearVelScale=<0.5 0.5 0.5>
    gravityOverride {z=-980}
    physicsBodyDefinitions [{
      boundBone {boneName=dyn_beltAcc_back_00}
      boxExtents=<15 15 15>
      constraintSetup {angularLimitsMax=<5 5 0>, angularLimitsMin=<-5 -5 -10>, angularTarget=<1 0 0>}
      localJointOffset=<4 0 0>
     }]
    relativeSpaceBone {boneName=pelvis}
    simSpaceSettings {maxAngularAcceleration=100, maxAngularVelocity=0.7, simSpaceAngularAlpha=1}
   }
   n5 = Anim Dynamics - Bone: dyn_beltAcc_back_01!(ComponentPose=@n4, Alpha=0.7, GravityOverride:Vector="0,0,-980") -> (Pose) {
    angularDampingOverride=1
    angularSpringConstant=1000
    bAngularSpring=true
    bOverrideAngularDamping=true
    bUseGravityOverride=true
    gravityScale=0.35
    simulationSpace=RootRelative
    boundBone {boneName=dyn_beltAcc_back_01}
    componentAppliedLinearAccClamp=<3000 3000 3000>
    componentLinearAccScale=<0.5 0.5 0.5>
    componentLinearVelScale=<0.5 0.5 0.5>
    gravityOverride {z=-980}
    physicsBodyDefinitions [{
      boundBone {boneName=dyn_beltAcc_back_01}
      boxExtents=<15 15 15>
      constraintSetup {angularLimitsMax=<5 5 0>, angularLimitsMin=<-5 -5 -10>, angularTarget=<1 0 0>}
      localJointOffset=<4 0 0>
     }]
    relativeSpaceBone {boneName=pelvis}
    simSpaceSettings {maxAngularAcceleration=100, maxAngularVelocity=0.7, simSpaceAngularAlpha=1}
   }
   n6 = Anim Dynamics - Bone: dyn_beltAcc_back_02!(ComponentPose=@n5, Alpha=0.7, GravityOverride:Vector="0,0,-980") -> (Pose) {
    angularDampingOverride=1
    angularSpringConstant=2000
    bAngularSpring=true
    bOverrideAngularDamping=true
    bUseGravityOverride=true
    gravityScale=0.35
    simulationSpace=RootRelative
    boundBone {boneName=dyn_beltAcc_back_02}
    componentAppliedLinearAccClamp=<3000 3000 3000>
    componentLinearAccScale=<0.5 0.5 0.5>
    componentLinearVelScale=<0.5 0.5 0.5>
    gravityOverride {z=-980}
    physicsBodyDefinitions [{
      boundBone {boneName=dyn_beltAcc_back_02}
      boxExtents=<15 15 15>
      constraintSetup {angularLimitsMax=<5 5 0>, angularLimitsMin=<-5 -5 -50>, angularTarget=<1 0 0>}
      localJointOffset=<4 0 0>
     }]
    relativeSpaceBone {boneName=pelvis}
    simSpaceSettings {maxAngularAcceleration=100, maxAngularVelocity=0.7, simSpaceAngularAlpha=1}
   }
   Output Pose(Result=Component To Local(ComponentPose=@n6)) {layerGroup=None}
  ]}
 EventGraph {nodes [
   Event BlueprintUpdateAnimation()
   > n1 = Sequence()
   > n2 = "Gravity Override State Function (Parameter Asset) (Nativized)"(
    CharacterPartAnimInstance=Self-Reference()
    GameTimeInSeconds=GetGameTimeInSeconds()
    Parameters=/BRCosmetics/Characters/Player/Female/Medium/Bodies/F_MED_CherryKnot_Strings/Meshes/Fortnite_CherryKnot_Strings_Parameters
   )
   > Set Gravity Override(Gravity Override=@n2)
   Set frontendAlpha(execute=@n1.then_1, frontendAlpha:double=SelectFloat(A=0.3, B=1, bPickA=Get bIsInFrontEnd())) -> (Output_Get:double)
  ]}
]
variables [
 frontendAlpha {type=number}
 Gravity Override {type=Vector}
]
